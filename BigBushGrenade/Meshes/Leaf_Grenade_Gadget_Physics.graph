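# Physics asset for the leaf grenade gadget mesh in BigBushGrenade, giving its five dynamic leaf bones simple physics bodies.
# Each leaf (top, bottom, front, left, right) has one rotated capsule, and the handle has a smaller capsule.
# Every leaf is constrained directly to the handle in a star layout, with narrow swing and tighter twist limits that allow only slight movement around it.
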PhysicsAsset /BigBushGrenade/Meshes/Leaf_Grenade_Gadget_Physics
bodies {
 dyn_leaf_bottom [{shapeType=Capsule, length=3.5, radius=3, shapeName=dyn_leaf_bottom_capsule, rotation=<rot 18.6979 -90 -3.8466>}]
 dyn_leaf_front [{shapeType=Capsule, length=3.5, radius=3, shapeName=dyn_leaf_front_capsule, rotation=<rot 0 -169.3979 -76.5167>}]
 dyn_leaf_left [{shapeType=Capsule, length=3.5, radius=3, shapeName=dyn_leaf_left_capsule, rotation=<rot 0 -2.0859 -48.9658>}]
 dyn_leaf_right [{shapeType=Capsule, length=3.5, radius=3, shapeName=dyn_leaf_right_capsule, rotation=<rot 23.6344 -90 21.6558>}]
 dyn_leaf_top [{shapeType=Capsule, length=3.5, radius=3, shapeName=dyn_leaf_top_capsule, rotation=<rot 40.913 -90 32.2659>}]
 handle [{shapeType=Capsule, length=1, radius=1}]
}
constraints [
 {bone1Name=dyn_leaf_bottom, bone2Name=handle, swing1LimitDegrees=18, swing2LimitDegrees=18, twistLimitDegrees=5}
 {bone1Name=dyn_leaf_front, bone2Name=handle, swing1LimitDegrees=18, swing2LimitDegrees=18, twistLimitDegrees=5}
 {bone1Name=dyn_leaf_left, bone2Name=handle, swing1LimitDegrees=18, swing2LimitDegrees=18, twistLimitDegrees=5}
 {bone1Name=dyn_leaf_right, bone2Name=handle, swing1LimitDegrees=18, swing2LimitDegrees=18, twistLimitDegrees=5}
 {bone1Name=dyn_leaf_top, bone2Name=handle, swing1LimitDegrees=18, swing2LimitDegrees=18, twistLimitDegrees=5}
]
